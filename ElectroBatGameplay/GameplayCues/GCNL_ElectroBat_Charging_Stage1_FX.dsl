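; Looping gameplay cue for the first ElectroBat charging stage, handling its effects and sound indicator.
; On looping start it prints a debug string, caches the target player pawn and feeds the per-charge timer into the first particle component.
; It starts a looping sound indicator for pawns off the local player's team and stops it on removal.
; Camera shake, force feedback and performance-check helpers are defined but not called here.
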(event EventOnLoopingStart(EitherVFX-preferthisversion)
  (params (MyTarget :type Actor%20Object%20Reference) (Parameters :type Gameplay%20Cue%20Parameters%20Structure%20%28by%20ref%29) (ParticleComponents :type Array%20of%20FXSystem%20Component%20Object%20References) (AudioComponents :type Array%20of%20Audio%20Component%20Object%20References))
  (|CallParentFunction MyTarget Parameters ParticleComponents AudioComponents)
  (Development|PrintString MyTarget "GC Started1" true true "(R=0.000000,G=0.660000,B=1.000000,A=1.000000)" 4.0)
  (bind _asfort_player_pawn_athena (Utilities|Casting|CastToFortPlayerPawnAthena MyTarget))
  (Variables|Default|SetPlayerPawn _asfort_player_pawn_athena)
  (Effects|Components|ParticleSystem|SetFloatParameter (Utilities|Array|Get(acopy) ParticleComponents) "GC Charging 2nd FX Start " (ScalableFloat|GetValueatLevel (Variables|Default|GetPer%20Charge%20Timer)))
  (if (not (Team|OnLocalPlayersTeam (Variables|Default|GetPlayerPawn) (Variables|Default|GetPlayerPawn)))
    (SoundIndicator|StartTargetedSoundIndicatorLooping (Variables|Default|GetFortSoundIndicator) (Variables|Default|GetPlayerPawn) 0 1.0 (Variables|/Script/Engine.Pawn|GetPlayerState (Variables|Default|GetPlayerPawn)))))

(event EventOnRemoval(EitherVFX-preferthisversion)
  (params (MyTarget :type Actor%20Object%20Reference) (Parameters :type Gameplay%20Cue%20Parameters%20Structure%20%28by%20ref%29) (ParticleComponents :type Array%20of%20FXSystem%20Component%20Object%20References) (AudioComponents :type Array%20of%20Audio%20Component%20Object%20References) (BurstCameraShakeInstance :type Camera%20Shake%20Base%20Object%20Reference) (BurstDecalInstance :type Decal%20Actor%20Object%20Reference))
  (|CallParentFunction MyTarget Parameters ParticleComponents AudioComponents BurstCameraShakeInstance BurstDecalInstance)
  (SoundIndicator|StopSoundIndicatorLooping (Variables|Default|GetFortSoundIndicator)))

(fn PlayCameraShake
  (params (TargetActor :type Actor%20Object%20Reference) (ShakeClass :type Camera%20Shake%20Base%20Class%20Reference) (Scale :type Float%20%28single-precision%29))
  (bind _returnvalue (Fortnite|GetFortPlayerControllerfromActor TargetActor))
  ; Play the desired camera shake on the owning player controller.
  (Utilities|IsValid _returnvalue
    (:Is%20Valid
      (bind _playercameramanager (Variables|/Script/Engine.PlayerController|GetPlayerCameraManager _returnvalue)
        (:then
          (CameraShakes|StartCameraShake :self _playercameramanager :ShakeClass ShakeClass :Scale Scale :PlaySpace "CameraLocal"))
        (:else)))))

(fn PlayForceFeedback
  (params (TargetActor :type Actor%20Object%20Reference) (ForceFeedbackEffect :type Force%20Feedback%20Effect%20Object%20Reference))
  (bind _returnvalue (Fortnite|GetFortPlayerControllerfromActor TargetActor))
  ;[0.008391 0.1308 0.1406] Play force feedback on the target actor (the player that is instigating the attack)
  (Utilities|IsValid TargetActor
    (:Is%20Valid
      (Utilities|IsValid _returnvalue
        (:Is%20Valid
          (Game|Feedback|LocalPlayForceFeedback _returnvalue ForceFeedbackEffect false "None"))))))

(fn PreferPerformance
  ;[1 0.3356 0.7917] Does the system prefer performance?
  (return (Math|Boolean|ORBoolean (Fortnite|IsMobilePlatform) (Class|PlatformHelpersLibrary|IsPcPerfMode self) (TouchPlatform|IsMobileGame self) (TouchPlatform|IsNintendoSwitchorNintendoSwitchPreview))))

(fn ConstructionScript)
