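; Gameplay cue burst notify that plays impact VFX and audio when an ElectroBat homerun swing hits a target.
; After the parent burst runs, it passes the hit's physical surface to the first particle component as the PhysicalSurface integer parameter.
; It resolves the instigator's player pawn and plays a sound at the target's location.
; The pawn's ChargedSwing.Hit tag picks homerun or uncharged audio, and local view picks the 1P or 3P variant.
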(fn OnBurst(EitherVFX-preferthisversion)
  (params (MyTarget :type Actor%20Object%20Reference) (Parameters :type Gameplay%20Cue%20Parameters%20Structure%20%28by%20ref%29) (ParticleComponents :type Array%20of%20FXSystem%20Component%20Object%20References) (AudioComponents :type Array%20of%20Audio%20Component%20Object%20References) (BurstCameraShakeInstance :type Camera%20Shake%20Base%20Object%20Reference) (BurstDecalInstance :type Decal%20Actor%20Object%20Reference))
  (bind (_normalizedmagnitude _rawmagnitude _effectcontext) (Ability|GameplayCue|BreakGameplayCueParameters Parameters))
  (bind _returnvalue (Transformation|GetActorLocation MyTarget))
  (|CallParentFunction MyTarget Parameters ParticleComponents AudioComponents BurstCameraShakeInstance BurstDecalInstance)
  (Effects|Components|ParticleSystem|SetIntParameter (Utilities|Array|Get(acopy) ParticleComponents) "PhysicalSurface" (Utilities|Select :Index (Collision|GetFortPhysicalSurface (Ability|EffectContext|GetHitResult _effectcontext)) :SurfaceType_Default 0 :SurfaceType1 1 :SurfaceType2 2 :SurfaceType3 3 :SurfaceType4 4 :SurfaceType5 5 :SurfaceType6 6 :SurfaceType7 7 :SurfaceType8 8 :SurfaceType9 9 :SurfaceType10 10 :SurfaceType11 11 :SurfaceType12 12 :SurfaceType13 13 :SurfaceType14 14 :SurfaceType15 15 :SurfaceType16 16 :SurfaceType17 17 :SurfaceType18 18 :SurfaceType19 19 :SurfaceType20 20 :SurfaceType21 21 :SurfaceType22 22 :SurfaceType23 23 :SurfaceType24 24 :SurfaceType25 25 :SurfaceType26 26 :SurfaceType27 0 :SurfaceType28 0 :SurfaceType29 0 :SurfaceType30 0 :SurfaceType31 0 :SurfaceType32 0 :SurfaceType33 0 :SurfaceType34 0))
  (bind _asfort_player_state (Utilities|Casting|CastToFortPlayerState (Ability|EffectContext|GetInstigatorActor _effectcontext))
    (:then
      (Utilities|IsValid (PlayerState|GetPawn _asfort_player_state)
        (:Is%20Valid
          (bind _asfort_player_pawn (Utilities|Casting|CastToFortPlayerPawn (PlayerState|GetPawn _asfort_player_state))
            (:then
              (if (Pawn|IsLocallyViewed _asfort_player_pawn)
                (Audio|PlaySoundatLocation _asfort_player_pawn (select (GameplayTags|HasMatchingGameplayTag _asfort_player_pawn "(TagName=\"Gameplay.Status.ElectroBat.ChargedSwing.Hit\")") (Variables|Default|GetHomerun1pSound) (Variables|Default|GetUncharged1pSound)) _returnvalue)
                (else
                  (Audio|PlaySoundatLocation _asfort_player_pawn (select (GameplayTags|HasMatchingGameplayTag _asfort_player_pawn "(TagName=\"Gameplay.Status.ElectroBat.ChargedSwing.Hit\")") (Variables|Default|GetHomerun3pSound) (Variables|Default|GetUncharged3pSound)) _returnvalue))))
            (:CastFailed
              _))
          (return))
        (:Is%20Not%20Valid
          _)))
    (:CastFailed
      _)))
